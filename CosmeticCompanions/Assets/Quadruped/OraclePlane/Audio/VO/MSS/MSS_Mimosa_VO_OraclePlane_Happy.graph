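# MetaSound Source preset that plays the happy-mood vocal reaction of the OraclePlane quadruped cosmetic companion (Mimosa VO).
# It inherits its graph from MSS_Mimosa_MoodReact and supplies seven happy-emotion waves, with probability, pitch and volume inputs all at one.
# Output is one-shot mono with a VO envelope float.
# Companion VO attenuation, a shared Mimosa property sheet, the pets sound class and two concurrency rules govern playback.
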MetaSoundSource /CosmeticCompanions/Assets/Quadruped/OraclePlane/Audio/VO/MSS/MSS_Mimosa_VO_OraclePlane_Happy
preset_parent=/FortMetasound/Cosmetics/Mimosa/Mood/MSS_Mimosa_MoodReact
inputs [
 Character.Affiliation {type=Enum:FortAudioAffiliation}
 Character.Health {type=Float}
 Character.UniqueID {type=Int32}
 Mood.Sounds {type=WaveAsset:Array}
 PitchModulationRange.Max {type=Float}
 PitchModulationRange.Min {type=Float}
 Probability {type=Float}
 TemplatePitchMultiplier {type=Float}
 TemplateVolumeMultiplier {type=Float}
 UE.Source.OnPlay {type=Trigger}
 VolumeModulationRange.Max {type=Float}
 VolumeModulationRange.Min {type=Float}
]
interfaces [Fort.Character:1.0, UE.OutputFormat.Mono:1.0, UE.Source.OneShot:1.0, UE.Source:1.1]
outputs [
 "UE.OutputFormat.Mono.Audio:0" {type=Audio}
 UE.Source.OneShot.OnFinished {type=Trigger}
 VoEnvelope {type=Float}
]
pages [Default {input_defaults {
   Character.Affiliation=0
   Character.Health=0
   Character.UniqueID=0
   PitchModulationRange.Max=1
   PitchModulationRange.Min=1
   Probability=1
   TemplatePitchMultiplier=1
   TemplateVolumeMultiplier=1
   UE.Source.OnPlay=null
   VolumeModulationRange.Max=1
   VolumeModulationRange.Min=1
   Mood.Sounds [
    /CosmeticCompanions/Assets/Quadruped/OraclePlane/Audio/VO/Waves/Mimosa_VO_OraclePlane_Emotion_Happy_010
    /CosmeticCompanions/Assets/Quadruped/OraclePlane/Audio/VO/Waves/Mimosa_VO_OraclePlane_Emotion_Happy_013
    /CosmeticCompanions/Assets/Quadruped/OraclePlane/Audio/VO/Waves/Mimosa_VO_OraclePlane_Emotion_Happy_014
    /CosmeticCompanions/Assets/Quadruped/OraclePlane/Audio/VO/Waves/Mimosa_VO_OraclePlane_Emotion_Happy_015
    /CosmeticCompanions/Assets/Quadruped/OraclePlane/Audio/VO/Waves/Mimosa_VO_OraclePlane_Emotion_Happy_020
    /CosmeticCompanions/Assets/Quadruped/OraclePlane/Audio/VO/Waves/Mimosa_VO_OraclePlane_Emotion_Happy_021
    /CosmeticCompanions/Assets/Quadruped/OraclePlane/Audio/VO/Waves/Mimosa_VO_OraclePlane_Emotion_Happy_022
   ]
  }}]
properties {
 assetImportData=&self:AssetImportData
 attenuationSettings=/Game/Sounds/Attenuation/AS_Companion_VO
 audioPropertiesSheet=/CosmeticCompanions/Assets/Audio_Shared/PropertySheets/APS_Cosmetic_Mimosa_VO
 maxDistance=2200
 sampleRate=48000
 soundClassObject=&Fort_Creatures_Pets
 assetUserData [&self:AssetUserData_0]
 concurrencySet [/Game/Sounds/Concurrency_Rules/CS_Mimosa_VO, /Game/Sounds/Concurrency_Rules/CS_Mimosa_VO_Global]
}
